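# Translucent unlit material for lightning effects on the Lyrical female glider, used on Niagara mesh particles.
# A sine-shaped core and noisy exterior are tinted by time of day and particle color, with switchable glow and optional desaturation.
# Dynamic parameters drive a deploy wipe along V and a smoke-noise dissolve in the opacity math.
# Panning cell noise, weighted by a sine along V, builds a local-space Y displacement.
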Material /BRCosmetics/Gadgets/Assets/VinderTech_GliderChute/Glider_Lyrical_Female/FX/M_Glider_Lyrical_female_Lightning
graph {
 properties {
  bUsedWithNiagaraMeshParticles=true
  blendMode=BLEND_Translucent
  pixelDepthOffsetMode=PDOM_Legacy
  shadingModel=MSM_Unlit
  usedShadingModels=Unlit
 }
 nodes [
  Constant() {r=0.5}
  MaterialFunctionCall() {materialFunction=&DebugTimeSine}
  n1 = Constant3Vector() {constant=<rgba 1 1 1 0>}
  n2 = ParticleColor()
  n3 = Sine(None=ComponentMask(None=TextureCoordinate()) {r=true}) {period=2}
  n4 = ScalarParameter() {defaultValue=10, parameterName=Center Power}
  n5 = TextureCoordinate() {uTiling=0.5, vTiling=0.8}
  n6 = VectorParameter() {parameterName=Exterior Color, defaultValue=<rgba 10 1.0915 0.2083 1>}
  n7 = Constant3Vector() {constant=<rgba 1 1 1 0>}
  n8 = MaterialFunctionCall(
   Color_Morning=@n7
   Color_Day=@n7
   Color_Evening=@n7
   Color_Night=VectorParameter() {parameterName=Color_Night, defaultValue=<rgba 0.6417 1 0.5249 0>}.RGB
  ) {materialFunction=&Color_Time_of_Day}
  n9 = Add(
   A=Multiply(A=VectorParameter() {parameterName=Internal_Color, defaultValue=<rgba 1 0.6679 0.0662 1>}.RGB, B=Power(Base=@n3, Exp=@n4) {constExponent=10})
   B=Multiply(
    A=TextureSample(UVs=@n5) {samplerType=SAMPLERTYPE_Grayscale, texture=&LowResBlurredNoise}.R
    B=Multiply(A=@n6.RGB, B=@n8.Result)
   )
  )
  n10 = ScalarParameter() {parameterName=Use_Glow}
  n11 = Multiply(A=VectorParameter() {parameterName=Glow_Color, defaultValue=<rgba 2 0.9891 0 1>}.RGB)
  n12 = Multiply(
   A=MaterialFunctionCall(
    Color_Morning=@n1
    Color_Day=@n1
    Color_Evening=@n1
    Color_Night=VectorParameter() {parameterName=Color_Night, defaultValue=<rgba 0.6417 1 0.5249 0>}.RGB
   ) {materialFunction=&Color_Time_of_Day}.Result
   B=Multiply(A=@n2.RGB, B=LinearInterpolate(A=@n9, B=@n11, Alpha=@n10))
  )
  n13 = DynamicParameter() {defaultValue=<rgba 0 1 1 0>, paramNames [Lightning_disappear, Lightning_deploy, Param3, Desaturation]}
  MaterialFunctionCall(Emissive=LinearInterpolate(A=@n12, B=Desaturation(None=@n12, Fraction=@n13.Desaturation), Alpha=ScalarParameter() {parameterName=Use Desaturation})) {materialFunction=&MF_FX_EmissiveAdjustments}
  n14 = ComponentMask(None=TextureCoordinate()) {g=true}
  Multiply(A=OneMinus(None=@n14))
  n15 = Power(Base=SphereMask(A=TextureCoordinate() {vTiling=0}, B=Constant2Vector() {r=0.5}) {attenuationRadius=0.7, hardnessPercent=0}) {constExponent=3}
  n16 = Constant() {r=0.15}
  n17 = LinearInterpolate(A=Multiply(A=@n2.A), B=Multiply(A=@n2.A, B=Multiply(A=@n15, B=@n16)), Alpha=ScalarParameter() {parameterName=Use_Glow})
  n18 = ComponentMask(None=TextureCoordinate()) {g=true}
  n19 = LinearInterpolate(Alpha=DynamicParameter() {defaultValue=<rgba 0 1 1 0>, paramNames [Lightning_disappear, Lightning_deploy, Param3, Desaturation]}.Lightning_deploy) {constA=-1}
  Multiply(A=@n17, B=Saturate(None=Add(A=@n18, B=@n19) {constB=0.5}))
  n20 = SphereMask(A=TextureCoordinate() {vTiling=0}, B=Constant2Vector() {r=0.5}) {attenuationRadius=0.5, hardnessPercent=0}
  n21 = LinearInterpolate(Alpha=TextureSample(UVs=TextureCoordinate() {uTiling=0.5}) {texture=&T_Inky_Smoke_Tile}.R) {constA=0.2}
  n22 = Constant()
  Multiply(
   A=@n17
   B=If(
    A=DynamicParameter() {defaultValue=<rgba 0 1 1 0>, paramNames [Lightning_disappear, Lightning_deploy, Param3, Desaturation]}.Lightning_disappear
    B=Multiply(A=@n20, B=@n21)
    "A > B"=@n22
    "A == B"=@n22
    "A < B"=Constant() {r=1}
   )
  )
  n23 = TextureCoordinate()
  Power(Base=Sine(None=ComponentMask(None=@n23) {r=true}) {period=2}) {constExponent=3}
  n24 = TextureCoordinate()
  n25 = Power(Base=Sine(None=ComponentMask(None=@n24) {r=true}) {period=2})
  Saturate(None=Multiply(A=@n25) {constB=5})
  Subtract(A=@n25, B=MaterialFunctionCall() {materialFunction=&DebugTimeSine}.Result)
  TextureSample()
  n26 = TextureSample(UVs=Panner(Coordinate=TextureCoordinate() {uTiling=0, vTiling=2}) {speedY=4}) {samplerType=SAMPLERTYPE_Grayscale, texture=/BRCosmetics/Weapons/FORT_Melee/Pickaxe_Cube_Ninja_Male/FX/T_CellNoise}
  n27 = Sine(None=@n14) {period=2}
  n28 = LinearInterpolate(A=Multiply(A=LinearInterpolate(Alpha=@n26.R), B=Saturate(None=@n27)))
  Transform(None=Multiply(A=Constant3Vector() {constant=<rgba 0 1 0 0>}, B=Multiply(A=@n28) {constB=100})) {transformSourceType=TRANSFORMSOURCE_Local}
 ]
}
