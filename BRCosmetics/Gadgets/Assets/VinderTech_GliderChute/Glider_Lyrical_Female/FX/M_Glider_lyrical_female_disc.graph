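# Translucent, lit glass-like material for the disc FX of the Lyrical Female glider, usable on skeletal and Niagara meshes.
# Glow combines noise-modulated masked lines, sphere-mask brightness and fake highlights from reflection-vector glow lookups and sun/moon direction.
# Time-of-day functions tint and desaturate it before emissive adjustment, with optional particle color.
# A tiled noise normal feeds Fresnel-based opacity, alongside index-of-refraction distortion and a Fresnel-lifted base color.
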Material /BRCosmetics/Gadgets/Assets/VinderTech_GliderChute/Glider_Lyrical_Female/FX/M_Glider_lyrical_female_disc
graph {
 properties {
  bEnableResponsiveAA=true
  bScreenSpaceReflections=true
  bUsedWithMeshDeformer=true
  bUsedWithNiagaraMeshParticles=true
  bUsedWithSkeletalMesh=true
  blendMode=BLEND_Translucent
  physMaterial=/Game/FortressPhysicalMaterials/Glass
  pixelDepthOffsetMode=PDOM_Legacy
  refractionMethod=RM_IndexOfRefraction
  translucencyDirectionalLightingIntensity=5
  translucencyLightingMode=TLM_VolumetricDirectional
  usedShadingModels=Default Lit
 }
 nodes [
  DynamicParameter() {defaultValue=<rgba 0 1 1 1>, paramNames [Param1,Param2,Param3,Desaturation]}
  n1 = Constant() {r=1}
  n2 = ParticleColor()
  n3 = ScalarParameter() {parameterName=Use_particle_color}
  n4 = TextureSample(UVs=TextureCoordinate() {coordinateIndex=1, uTiling=0.5, vTiling=0.5}) {samplerType=SAMPLERTYPE_Grayscale, texture=&LowResBlurredNoise}
  n5 = TextureSample() {
   samplerType=SAMPLERTYPE_Masks
   texture=/BRCosmetics/Gadgets/Assets/VinderTech_GliderChute/Glider_Lyrical_Female/Textures/T_Glider_Lyrical_Female_FX
  }
  n6 = Multiply(A=VectorParameter() {parameterName=Line color, defaultValue=<rgba 10 7.7981 0.3484 1>}.RGB, B=@n5.R) {constA=5}
  n7 = TextureCoordinate() {coordinateIndex=1, uTiling=1.1}
  n8 = Constant2Vector() {g=0.3, r=0.55}
  n9 = Constant() {r=0.55}
  n10 = Multiply(
   A=Multiply(A=LinearInterpolate(Alpha=@n4.R) {constA=-0.5}, B=Power(Base=@n6))
   B=Power(Base=SphereMask(A=@n7, B=@n8, Radius=@n9) {attenuationRadius=0.5, hardnessPercent=0})
  )
  n11 = ComponentMask(None=Transform(None=ReflectionVectorWS()) {transformSourceType=TRANSFORMSOURCE_World, transformType=TRANSFORM_View}) {g=true, r=true}
  n12 = TextureSample(UVs=Multiply(A=ConstantBiasScale(None=@n11), B=Constant2Vector() {g=1.2, r=1.2})) {texture=/Game/Weapons/FORT_Melee/Effects/LaserSwordFounders/T_LaserSwordGlow}
  n13 = CollectionParameter() {
   collection=&FortniteMaterialParameters
   parameterName=SunAndMoonModelDirectionalVector
  }
  n14 = Clamp(None=DotProduct(A=ComponentMask(None=@n13) {b=true, g=true, r=true}, B=ReflectionVectorWS()))
  n15 = Constant() {r=70}
  n16 = Add(
   A=Multiply(A=Power(Base=@n12.R) {constExponent=20}, B=Constant() {r=0.7})
   B=Multiply(A=Power(Base=@n14, Exp=@n15) {constExponent=20}, B=Constant() {r=3})
  ) {constB=0}
  n17 = TextureCoordinate() {coordinateIndex=1, uTiling=1.1}
  n18 = Constant2Vector() {g=0.3, r=0.55}
  n19 = Constant() {r=45}
  n20 = OneMinus(None=Power(Base=SphereMask(A=@n17, B=@n18, Radius=@n19) {attenuationRadius=0.5, hardnessPercent=0}) {constExponent=1})
  n21 = Multiply(
   A=VectorParameter() {parameterName=Brightness color, defaultValue=<rgba 10 9.4098 0 1>}.RGB
   B=Saturate(None=LinearInterpolate(Alpha=@n20) {constA=-0.025, constB=10})
  ) {constA=30}
  n22 = Multiply(A=@n21, B=@n21)
  n23 = Saturate(None=Multiply(A=@n22, B=@n22))
  n24 = Add(
   A=Add(A=Add(A=@n10, B=@n16), B=Power(Base=@n23)) {constB=0}
   B=VectorParameter() {parameterName=Dark color, defaultValue=<rgba 0.1302 0.0908 0 1>}.RGB
  )
  n25 = Constant() {r=-1}
  n26 = Multiply(
   A=Multiply(A=LinearInterpolate(A=@n1, B=@n2.RGB, Alpha=@n3), B=Desaturation(None=@n24, Fraction=@n25))
   B=MaterialFunctionCall(
    Color_Morning=VectorParameter() {parameterName=Color_Morning, defaultValue=<rgba 0.3021 0.3021 0.3021 0>}.RGB
    Color_Day=Constant() {r=1}
    Color_Evening=VectorParameter() {parameterName=Color_Evening, defaultValue=<rgba 0.3177 0.3177 0.3177 0>}.RGB
    Color_Night=VectorParameter() {parameterName=Color_Night, defaultValue=<rgba 1 1 1 0>}.RGB
   ) {materialFunction=&Color_Time_of_Day}.Result
  )
  n27 = MaterialFunctionCall(Value_Morning=Constant(), Value_Day=Constant(), Value_Evening=Constant(), Value_Night=ScalarParameter() {parameterName="Value_Night (S)"}) {materialFunction=&Scalar_Time_of_Day}
  MaterialFunctionCall(Emissive=LinearInterpolate(A=@n26, B=Desaturation(None=@n26, Fraction=@n27.Result), Alpha=ScalarParameter() {parameterName=Use Desaturation})) {materialFunction=&MF_FX_EmissiveAdjustments}
  n28 = Constant() {r=0.2}
  n29 = Constant() {r=0.85}
  n30 = VectorParameter() {group=Normal, parameterName=Normal_UV_scale, channelNames {a="", b="", g=V Scale, r=U Scale}, defaultValue=<rgba 0.5 0 0 1>}
  n31 = TextureSample(UVs=Multiply(A=TextureCoordinate(), B=AppendVector(A=@n30.U Scale, B=@n30.U Scale))) {samplerType=SAMPLERTYPE_Normal, texture=/Game/Characters/Player/Female/Medium/Bodies/F_MED_Bistro_Astronaut/FX/T_GaussianNoise_Normal}
  n32 = Normalize(VectorInput=Multiply(A=@n31.RGB, B=Constant3Vector() {constant=<rgba 0.1 0.1 1 0>}))
  n33 = Saturate(None=Fresnel(ExponentIn=Constant() {r=3}, Normal=Transform(None=@n32)))
  Multiply(
   A=Multiply(A=LinearInterpolate(A=@n28, B=@n29, Alpha=@n33)) {constB=2}
   B=LinearInterpolate(A=Constant() {r=1}, B=ParticleColor().A, Alpha=ScalarParameter() {parameterName=Use_particle_color})
  )
  Reroute(None=@n32)
  Reroute(None=ScalarParameter() {defaultValue=1.005, parameterName=Refraction, sliderMax=2})
  n34 = Constant() {r=0.5}
  n35 = Fresnel(ExponentIn=Constant() {r=3})
  Reroute(None=Add(A=Multiply(A=@n34, B=@n35), B=VectorParameter() {parameterName=Base color, defaultValue=<rgba 0.276 0.1213 0 1>}.RGB))
  TextureCoordinate()
 ]
}
