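# Translucent unlit Niagara mesh-particle material for the sound wave ring effect on the Lyrical female glider.
# Panning smoke noise, base color, particle color and a time-of-day tint form the emissive, with optional desaturation driven by a dynamic parameter.
# Spiky displacement radiates from the object position, combining curve-atlas profile, time sine, panning noise and a hard sine/cell-noise mask.
# Dynamic parameters scale its intensity and shift the spike pattern in floored steps.
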Material /BRCosmetics/Gadgets/Assets/VinderTech_GliderChute/Glider_Lyrical_Female/FX/M_Lyrical_Glider_female_SoundWave_Ring
graph {
 properties {
  bUsedWithNiagaraMeshParticles=true
  blendMode=BLEND_Translucent
  pixelDepthOffsetMode=PDOM_Legacy
  shadingModel=MSM_Unlit
  usedShadingModels=Unlit
 }
 nodes [
  n1 = Constant3Vector() {constant=<rgba 1 1 1 0>}
  n2 = VectorParameter() {parameterName=Base_Color, defaultValue=<rgba 5 2.0212 0 1>}
  n3 = Panner(Coordinate=TextureCoordinate() {uTiling=0.5}) {speedY=0.5}
  n4 = Power(Base=LinearInterpolate(Alpha=TextureSample(UVs=@n3) {texture=&T_Inky_Smoke_Tile}.R) {constA=0.1}) {constExponent=1.3}
  n5 = Multiply(
   A=MaterialFunctionCall(
    Color_Morning=@n1
    Color_Day=@n1
    Color_Evening=@n1
    Color_Night=VectorParameter() {parameterName=Color_Night, defaultValue=<rgba 0.7955 1 0.2496 0>}.RGB
   ) {materialFunction=&Color_Time_of_Day}.Result
   B=Multiply(A=Multiply(A=@n2.RGB, B=@n4), B=ParticleColor().RGB)
  )
  n6 = DynamicParameter() {defaultValue=<rgba 0 1 1 0>, paramNames [spikes offset, WPO_intensity, Param3, Desaturation]}
  MaterialFunctionCall(Emissive=LinearInterpolate(A=@n5, B=Desaturation(None=@n5, Fraction=@n6.Desaturation), Alpha=ScalarParameter() {parameterName=Use Desaturation})) {materialFunction=&MF_FX_EmissiveAdjustments}
  n7 = WorldPosition()
  n8 = ObjectPositionWS()
  n9 = Floor(None=DynamicParameter() {defaultValue=<rgba 0 1 1 0>, paramNames [spikes offset, WPO_intensity, Param3, Desaturation]}.spikes offset)
  n10 = ComponentMask(None=Add(A=TextureCoordinate(), B=Multiply(A=@n9) {constB=0.1})) {g=true}
  n11 = CurveAtlasRowParameter(CurveTime=Frac(None=Multiply(A=@n10) {constB=10})) {
   atlas=/Game/Athena/LinearColorCurves/A_Atlas_Turtleneck_Glider
   curve=/Game/Gadgets/Assets/VinderTech_GliderChute/Glider_Lyrical_Female/FX/Curve_Glider_Lyrical_Female
   defaultValue=9
  }
  n12 = Multiply(A=LinearInterpolate(B=LinearInterpolate(Alpha=@n11.A) {constA=-0.5, constB=3})) {constB=18}
  n13 = MaterialFunctionCall(Speed=Constant() {r=2}) {materialFunction=&DebugTimeSine}
  n14 = TextureSample(UVs=Panner(Coordinate=TextureCoordinate()) {speedY=2}) {texture=&T_Inky_Smoke_Tile}
  n15 = Multiply(
   A=LinearInterpolate(Alpha=Multiply(A=@n13.Result, B=@n14.R)) {constA=-0.7, constB=3.3}
   B=DynamicParameter() {defaultValue=<rgba 0 1 1 0>, paramNames [spikes offset, WPO_intensity, Param3, Desaturation]}.WPO_intensity
  )
  n16 = Floor(None=DynamicParameter() {defaultValue=<rgba 0 1 1 0>, paramNames [spikes offset, WPO_intensity, Param3, Desaturation]}.spikes offset)
  n17 = ComponentMask(None=Add(A=TextureCoordinate() {vTiling=3}, B=Multiply(A=@n16) {constB=0.1})) {g=true}
  n18 = Floor(None=DynamicParameter() {defaultValue=<rgba 0 1 1 0>, paramNames [spikes offset, WPO_intensity, Param3, Desaturation]}.spikes offset)
  n19 = TextureSample(UVs=Add(A=TextureCoordinate() {uTiling=0, vTiling=2}, B=Multiply(A=@n18) {constB=0.1})) {samplerType=SAMPLERTYPE_Grayscale, texture=/BRCosmetics/Weapons/FORT_Melee/Pickaxe_Cube_Ninja_Male/FX/T_CellNoise}
  n20 = Ceil(None=Multiply(A=Sine(None=@n17) {period=0.52}, B=LinearInterpolate(Alpha=@n19.R) {constA=-0.2}))
  Multiply(A=Normalize(VectorInput=Subtract(A=@n7.XYZ, B=@n8)), B=Multiply(A=Multiply(A=@n12, B=@n15), B=Saturate(None=@n20)))
 ]
}
