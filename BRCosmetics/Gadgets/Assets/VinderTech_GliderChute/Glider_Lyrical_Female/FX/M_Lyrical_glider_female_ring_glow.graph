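# Unlit translucent, two-sided Niagara sprite and mesh particle material for the ring and glow FX of the Lyrical female glider.
# Scalar parameters blend procedural SphereMask shapes (U-only falloff, thin ring, soft glow) or full coverage into opacity, with particle alpha and DepthFade.
# Emissive multiplies particle color, optional randomly offset blurred noise and a time-of-day tint with a night color parameter.
# Optional dynamic-parameter desaturation precedes shared FX emissive adjustments.
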Material /BRCosmetics/Gadgets/Assets/VinderTech_GliderChute/Glider_Lyrical_Female/FX/M_Lyrical_glider_female_ring_glow
graph {
 properties {
  bUsedWithNiagaraMeshParticles=true
  bUsedWithNiagaraSprites=true
  blendMode=BLEND_Translucent
  pixelDepthOffsetMode=PDOM_Legacy
  shadingModel=MSM_Unlit
  twoSided=true
  usedShadingModels=Unlit
 }
 nodes [
  n1 = ParticleColor()
  n2 = Power(Base=SphereMask(A=TextureCoordinate() {vTiling=0}, B=Constant2Vector() {r=0.5}) {attenuationRadius=0.7, hardnessPercent=0}) {constExponent=8}
  n3 = SphereMask(A=TextureCoordinate(), B=Constant() {r=0.5}) {attenuationRadius=0.45, hardnessPercent=0}
  n4 = Multiply(
   A=Power(Base=OneMinus(None=@n3)) {constExponent=20}
   B=SphereMask(A=TextureCoordinate(), B=Constant() {r=0.5}) {attenuationRadius=0.5, hardnessPercent=0}
  )
  n5 = ScalarParameter() {parameterName=Use_Sprite_ring}
  n6 = SphereMask(A=TextureCoordinate(), B=Constant() {r=0.5}) {attenuationRadius=0.5, hardnessPercent=0}
  n7 = LinearInterpolate(
   A=LinearInterpolate(A=LinearInterpolate(A=@n2, B=@n4, Alpha=@n5), B=Power(Base=@n6) {constExponent=0.8}, Alpha=ScalarParameter() {parameterName=Use_sample_glow})
   B=Constant() {r=1}
   Alpha=ScalarParameter() {parameterName=Use_gradients_Texture}
  )
  DepthFade(
   Opacity=LinearInterpolate(A=Multiply(A=@n1.A, B=@n7), B=ParticleColor().A, Alpha=ScalarParameter() {parameterName=Use simple particlecolor})
   FadeDistance=ScalarParameter() {parameterName=DepthFade}
  )
  n8 = Constant3Vector() {constant=<rgba 1 1 1 0>}
  n9 = TextureCoordinate() {uTiling=0.5, vTiling=0.5}
  n10 = ParticleRandom()
  n11 = LinearInterpolate(
   A=Constant() {r=1}
   B=TextureSample(UVs=Add(A=@n9, B=@n10)) {samplerType=SAMPLERTYPE_Grayscale, texture=&LowResBlurredNoise}.R
   Alpha=ScalarParameter() {parameterName=Use_gradients_Texture}
  )
  n12 = Multiply(
   A=MaterialFunctionCall(
    Color_Morning=@n8
    Color_Day=@n8
    Color_Evening=@n8
    Color_Night=VectorParameter() {parameterName=Color_Night, defaultValue=<rgba 0.7955 1 0.2496 0>}.RGB
   ) {materialFunction=&Color_Time_of_Day}.Result
   B=LinearInterpolate(A=Multiply(A=@n11, B=@n1.RGB), B=ParticleColor().RGB, Alpha=ScalarParameter() {parameterName=Use simple particlecolor})
  )
  n13 = DynamicParameter() {defaultValue=<rgba 1 1 1 0>, paramNames [Param1,Param2,Param3,Des]}
  MaterialFunctionCall(Emissive=LinearInterpolate(A=@n12, B=Desaturation(None=@n12, Fraction=@n13.Des), Alpha=ScalarParameter() {parameterName=Use Desaturation})) {materialFunction=&MF_FX_EmissiveAdjustments}
 ]
}
